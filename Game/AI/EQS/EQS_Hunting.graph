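# EQS query that picks a navmesh location for hunting AI relative to its goal.
# A simple grid is generated around a hunting context location and projected to navmesh, with parameter-bound size and spacing.
# Two opposing distance-to-goal scores combine: a clamped one favors farther points, a stronger unclamped one favors nearer points.
# A batched pathfinding test checks candidates using the Hunting navigation filter.
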EnvQuery /Game/AI/EQS/EQS_Hunting
options [{
  generator {
   name=EnvQueryGenerator_SimpleGrid_2
   class=EnvQueryGenerator_SimpleGrid
   properties {
    generateAround=/Game/AI/EQS/Contexts/HuntingGeneratorLocationContext.HuntingGeneratorLocationContext_C
    gridSize {dataBinding=&self:EnvQueryGenerator_SimpleGrid_2.AIDataProvider_QueryParams_0, dataField=FloatValue, defaultValue=768}
    projectionData {
     extentX=10
     navigationFilter=&RecastFilter_UseDefaultArea
     postProjectionVerticalOffset=30
     projectDown=1024
     projectUp=1024
    }
    spaceBetween {dataBinding=&self:EnvQueryGenerator_SimpleGrid_2.AIDataProvider_QueryParams_1, dataField=FloatValue, defaultValue=150}
   }
  }
  tests [
   EnvQueryTest_Distance_0 {
    class=EnvQueryTest_Distance
    properties {
     clampMaxType=SpecifiedValue
     clampMinType=SpecifiedValue
     distanceTo=&FortQueryContext_Goal
     filterType=Range
     floatValueMin=0
     scoringEquation=Linear
     scoringFactor=0.4
     testComment=""
     testMode=Distance3D
     testPurpose=Score
     scoreClampMax {dataBinding=&self:EnvQueryTest_Distance_0.AIDataProvider_QueryParams_3, dataField=FloatValue, defaultValue=768}
     scoreClampMin {dataBinding=&self:EnvQueryTest_Distance_0.AIDataProvider_QueryParams_2, dataField=FloatValue, defaultValue=100}
    }
   }
   EnvQueryTest_Distance_42 {
    class=EnvQueryTest_Distance
    properties {
     distanceTo=&FortQueryContext_Goal
     filterType=Range
     floatValueMin=0
     scoringEquation=Linear
     scoringFactor=-0.6
     testComment=""
     testMode=Distance3D
     testPurpose=Score
    }
   }
   EnvQueryTest_PathfindingBatch_0 {class=EnvQueryTest_PathfindingBatch, properties {filterClass=/Script/FortniteAI.FortNavigationFilter_Hunting, testComment=""}}
  ]
 }]
